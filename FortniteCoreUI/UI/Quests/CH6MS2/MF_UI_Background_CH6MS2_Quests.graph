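# Material function that renders an animated, layered cloud-sky background for the CH6MS2 Quests UI, with day and night colour palettes.
# Five cloud layers scroll at separate speeds, tiling VaultWorld-derived UVs into cells with per-cell random values.
# MSDF textures supply antialiased RGB-channel cloud masks, composited through chained lerps with threshold gating.
# Gradients and an eased radial vignette finish the colour output, with VaultWorld focal-point helpers providing transition and focal mask.
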MaterialFunction /FortniteCoreUI/UI/Quests/CH6MS2/MF_UI_Background_CH6MS2_Quests
graph {
 properties {previewMaterialDomain=MD_UI}
 nodes [
  MaterialExpressionAdd_34 = Add(A=NamedRerouteUsage(), B=AppendVector(A=Constant(), B=Constant() {r=-0.108}))
  n1 = Convert(None=@MaterialExpressionAdd_34) {
   nodeName=Break Float2
   convertInputs [{
     type=Vector2
     expressionInput {
      expression=&self:MaterialExpressionAdd_34
      inputName=None
      mask=0
      maskA=0
      maskB=0
      maskG=0
      maskR=0
      outputIndex=0
     }
    }]
   convertOutputs [{type=Scalar}, {type=Scalar}]
  }
  AppendVector(A=DDX(Value=@n1), B=DDY(Value=@n1))
  MaterialExpressionAdd_33 = Add(A=NamedRerouteUsage(), B=AppendVector(A=Constant(), B=Constant() {r=-0.032}))
  n2 = Convert(None=@MaterialExpressionAdd_33) {
   nodeName=Break Float2
   convertInputs [{
     type=Vector2
     expressionInput {
      expression=&self:MaterialExpressionAdd_33
      inputName=None
      mask=0
      maskA=0
      maskB=0
      maskG=0
      maskR=0
      outputIndex=0
     }
    }]
   convertOutputs [{type=Scalar}, {type=Scalar}]
  }
  AppendVector(A=DDX(Value=@n2), B=DDY(Value=@n2))
  MaterialExpressionAdd_32 = Add(A=NamedRerouteUsage(), B=AppendVector(A=Constant(), B=Constant() {r=-0.106}))
  n3 = Convert(None=@MaterialExpressionAdd_32) {
   nodeName=Break Float2
   convertInputs [{
     type=Vector2
     expressionInput {
      expression=&self:MaterialExpressionAdd_32
      inputName=None
      mask=0
      maskA=0
      maskB=0
      maskG=0
      maskR=0
      outputIndex=0
     }
    }]
   convertOutputs [{type=Scalar}, {type=Scalar}]
  }
  AppendVector(A=DDX(Value=@n3), B=DDY(Value=@n3))
  n4 = NamedRerouteUsage()
  n5 = Subtract(A=Max(A=NamedRerouteUsage()) {constB=0}, B=Constant() {r=0.76})
  n6 = NamedRerouteUsage()
  n7 = NamedRerouteUsage()
  n8 = LinearInterpolate(A=@n7, B=NamedRerouteUsage())
  n9 = LinearInterpolate(A=@n6, B=@n8, Alpha=NamedRerouteUsage())
  n10 = LinearInterpolate(A=@n7, B=NamedRerouteUsage())
  n11 = NamedRerouteUsage()
  n12 = LinearInterpolate(
   A=@n6
   B=LinearInterpolate(A=LinearInterpolate(A=@n9, B=@n10, Alpha=@n11), B=@n8, Alpha=NamedRerouteUsage())
   Alpha=Ceil(None=Saturate(None=@n5))
  )
  n13 = NamedRerouteUsage()
  n14 = LinearInterpolate(A=@n13, B=NamedRerouteUsage()) {constAlpha=0.6}
  n15 = LinearInterpolate(A=@n12, B=@n14, Alpha=NamedRerouteUsage())
  n16 = LinearInterpolate(A=@n13, B=NamedRerouteUsage()) {constAlpha=0.6}
  n17 = NamedRerouteUsage()
  n18 = Subtract(A=Max(A=NamedRerouteUsage()) {constB=0}, B=Constant() {r=0.536})
  n19 = LinearInterpolate(
   A=@n12
   B=LinearInterpolate(A=LinearInterpolate(A=@n15, B=@n16, Alpha=@n17), B=@n14, Alpha=NamedRerouteUsage())
   Alpha=Ceil(None=Saturate(None=@n18))
  )
  n20 = Subtract(A=Max(A=NamedRerouteUsage()) {constB=0}, B=Constant() {r=0.824})
  n21 = NamedRerouteUsage()
  n22 = LinearInterpolate(A=@n21, B=NamedRerouteUsage()) {constAlpha=0.7}
  n23 = LinearInterpolate(A=@n19, B=@n22, Alpha=NamedRerouteUsage())
  n24 = LinearInterpolate(A=@n21, B=NamedRerouteUsage()) {constAlpha=0.7}
  n25 = NamedRerouteUsage()
  n26 = LinearInterpolate(
   A=@n19
   B=LinearInterpolate(A=LinearInterpolate(A=@n23, B=@n24, Alpha=@n25), B=@n22, Alpha=NamedRerouteUsage())
   Alpha=Ceil(None=Saturate(None=@n20))
  )
  n27 = NamedRerouteUsage()
  n28 = LinearInterpolate(A=@n27, B=NamedRerouteUsage()) {constAlpha=0.85}
  n29 = LinearInterpolate(A=@n26, B=@n28, Alpha=NamedRerouteUsage())
  n30 = LinearInterpolate(A=@n27, B=NamedRerouteUsage()) {constAlpha=0.85}
  n31 = NamedRerouteUsage()
  n32 = Subtract(A=Max(A=NamedRerouteUsage()) {constB=0}, B=Constant() {r=0.7605})
  n33 = LinearInterpolate(
   A=@n26
   B=LinearInterpolate(A=LinearInterpolate(A=@n29, B=@n30, Alpha=@n31), B=@n28, Alpha=NamedRerouteUsage())
   Alpha=Ceil(None=Saturate(None=@n32))
  )
  n34 = Subtract(A=Max(A=NamedRerouteUsage()) {constB=0}, B=Constant() {r=0.792})
  n35 = NamedRerouteUsage()
  n36 = LinearInterpolate(A=@n33, B=@n35, Alpha=NamedRerouteUsage())
  n37 = NamedRerouteUsage()
  n38 = NamedRerouteUsage()
  n39 = LinearInterpolate(
   A=@n33
   B=LinearInterpolate(A=LinearInterpolate(A=@n36, B=@n37, Alpha=@n38), B=@n35, Alpha=NamedRerouteUsage())
   Alpha=Ceil(None=Saturate(None=@n34))
  )
  n40 = Rotator(Coordinate=NamedRerouteUsage(), Time=Constant())
  n41 = MaterialFunctionCall(Input=Saturate(None=ComponentMask(None=@n40) {g=true}), Input Low=Constant() {r=0.75}, Input High=Constant()) {materialFunction=&RemapValue_0-1}
  n42 = LinearInterpolate(
   A=LinearInterpolate(A=LinearInterpolate(A=@n4, B=@n39, Alpha=@n41.Result), B=NamedRerouteUsage(), Alpha=NamedRerouteUsage())
   B=NamedRerouteUsage()
   Alpha=NamedRerouteUsage()
  )
  n43 = LinearInterpolate(A=Constant() {r=0.674}, B=Constant(), Alpha=NamedRerouteUsage())
  n44 = Constant() {r=3.5144}
  n45 = Constant() {r=2.9501}
  n46 = NamedRerouteUsage()
  n47 = MaterialFunctionCall(UV=MaterialFunctionCall(UVs=NamedRerouteUsage(), Texture Scale=LinearInterpolate(A=@n44, B=@n45, Alpha=@n46)) {materialFunction=&ScaleUVsByCenter}.UVs) {materialFunction=&MF_UI_RadialGradient}
  n48 = LinearInterpolate(
   A=Multiply(A=NamedRerouteUsage(), B=Constant() {r=2.2739})
   B=@n42
   Alpha=NamedRerouteDeclaration(None=MaterialFunctionCall("Value (0-1)"=@n47.Result) {materialFunction=&MF_UI_EaseCurves}.EaseInSine) {name=Vignette Mask, nodeColor=<rgba 0 1 0.9531 1>}
  )
  FunctionOutput(None=Desaturation(None=LinearInterpolate(A=@n42, B=@n48, Alpha=@n43), Fraction=Constant()))
  MaterialFunctionCall() {materialFunction=&GetUserInterfaceUV}
  n49 = Add(A=NamedRerouteUsage(), B=AppendVector(A=Constant(), B=Constant() {r=0.12}))
  n50 = TextureSample(UVs=@n49) {texture=/FortniteCoreUI/UI/Quests/CH6MS2/Textures/T_UI_Quests_CH6MS2_Cloud01_MSDF}
  MaterialFunctionCall(DistanceField=@n50.R, Threshold=Constant() {r=0.28}, Smoothing=Constant()) {materialFunction=&MF_UI_AntiAliasing}
  MaterialFunctionCall(DistanceField=@n50.G, Threshold=Constant() {r=0.28}, Smoothing=Constant()) {materialFunction=&MF_UI_AntiAliasing}
  MaterialFunctionCall(DistanceField=@n50.B, Threshold=Constant() {r=0.368}, Smoothing=Constant() {r=0.22}) {materialFunction=&MF_UI_AntiAliasing}
  Multiply(A=NamedRerouteUsage(), B=ScalarParameter() {parameterName=NightRadial_Darken}) {constB=0.5}
  n51 = MaterialFunctionCall!(Vault World Color, Focal Color Override, Focal Scale Override=Constant(), Mask Blend) {materialFunction=/Game/Environments/FrontEnd/Blueprints/ItemPreview/VaultWorldTransitions/MaterialFunctions/MF_UI_VaultWorld_FocalPoint}
  NamedRerouteDeclaration(None=@n51.Focal Mask) {name=FocalMask, nodeColor=<rgba 0 0 0 1>}
  NamedRerouteDeclaration(None=@n51.Focal Point UV) {name=FocalPointUV, nodeColor=<rgba 0 0 0 1>}
  NamedRerouteDeclaration(None=Constant3Vector() {constant=<rgba 0.1195 0.624 1 1>}) {name=ColourBGRadial_Day_01, nodeColor=<rgba 0 0 0 1>}
  NamedRerouteDeclaration(None=Constant3Vector() {constant=<rgba 0.0027 0.0497 0.2831 1>}) {name=ColourBGRadial_Day_02, nodeColor=<rgba 0 0 0 1>}
  n52 = MaterialFunctionCall(UV Channel=NamedRerouteUsage()) {materialFunction=&LinearGradient}
  n53 = Constant() {r=0.5}
  n54 = Constant() {r=1}
  NamedRerouteDeclaration(None=LinearInterpolate(
    A=NamedRerouteUsage()
    B=NamedRerouteUsage()
    Alpha=MaterialFunctionCall(Input=@n52.VGradient, Input Low=@n53, Input High=@n54) {materialFunction=&RemapValue_0-1}.Result
   )) {name=LinearBG, nodeColor=<rgba 0 0 0 1>}
  NamedRerouteDeclaration(None=Constant3Vector() {constant=<rgba 0.0252 0.0144 0.0844 1>}) {name=ColourBG_Night_04, nodeColor=<rgba 0 0 0 1>}
  NamedRerouteDeclaration(None=Constant3Vector() {constant=<rgba 0.0497 0.0284 0.1274 1>}) {name=ColourBG_Night_01, nodeColor=<rgba 0 0 0 1>}
  NamedRerouteDeclaration(None=Constant3Vector() {constant=<rgba 0.0176 0.0307 0.162 1>}) {name=ColourBG_Night_02, nodeColor=<rgba 0 0 0 1>}
  n55 = Constant3Vector() {constant=<rgba 0.0274 0.1234 0.4115 1>}
  NamedRerouteDeclaration(None=@n55) {name=ColourBG_Night_03, nodeColor=<rgba 0 0 0 1>}
  NamedRerouteDeclaration(None=Constant3Vector() {constant=<rgba 0.0252 0.0685 0.2582 1>}) {name=ColourBGRadial_Night_01, nodeColor=<rgba 0 0 0 1>}
  NamedRerouteDeclaration(None=Constant3Vector() {constant=<rgba 0.0252 0.007 0.1845 1>}) {name=ColourBGRadial_Night_02, nodeColor=<rgba 0 0 0 1>}
  NamedRerouteDeclaration(None=Constant3Vector() {constant=<rgba 0.0144 0.0296 0.1559 1>}) {name=ColourCloud_Night_01, nodeColor=<rgba 0 0 0 1>}
  n56 = MaterialFunctionCall!(Use Local UV=FunctionInput(Preview=Constant()) {inputName=BGLocalUV}, Texture Size, Scale Pivot Point) {materialFunction=/Game/Environments/FrontEnd/Blueprints/ItemPreview/VaultWorldTransitions/MaterialFunctions/MF_UI_VaultWorld_UV}
  NamedRerouteDeclaration(None=@n56.UV) {name=UV, nodeColor=<rgba 0 0 0 1>}
  NamedRerouteDeclaration(None=Constant3Vector() {constant=<rgba 0.0723 0.0482 0.227 1>}) {name=ColourCloud_Night_02, nodeColor=<rgba 0 0 0 1>}
  NamedRerouteDeclaration(None=@n56.Scale Y) {name=ScaleY, nodeColor=<rgba 0 0 0 1>}
  n57 = NamedRerouteUsage()
  n58 = LinearInterpolate(A=Constant() {r=12.57}, B=Constant() {r=12.57}, Alpha=NamedRerouteUsage())
  n59 = Saturate(None=ComponentMask(None=Rotator(Coordinate=@n57, Time=@n58)) {g=true})
  n60 = LinearInterpolate(A=Constant() {r=0.2131}, B=Constant() {r=0.05}, Alpha=NamedRerouteUsage())
  n61 = LinearInterpolate(A=Constant() {r=1.0555}, B=Constant() {r=1.136}, Alpha=NamedRerouteUsage())
  NamedRerouteDeclaration(None=Power(
    Base=MaterialFunctionCall(Input=@n59, Input Low=@n60, Input High=@n61) {materialFunction=&RemapValue_0-1}.Result
    Exp=Constant() {r=2}
   )) {name=Top Linear Gradient, nodeColor=<rgba 0 0 0 1>}
  n62 = NamedRerouteUsage()
  n63 = LinearInterpolate(A=Constant() {r=-0.3}, B=Constant(), Alpha=NamedRerouteUsage())
  n64 = Saturate(None=ComponentMask(None=Rotator(Coordinate=@n62, Time=@n63)) {g=true})
  n65 = LinearInterpolate(A=Constant() {r=0.3788}, B=Constant() {r=0.3548}, Alpha=NamedRerouteUsage())
  n66 = LinearInterpolate(A=Constant() {r=0.9953}, B=Constant() {r=0.9676}, Alpha=NamedRerouteUsage())
  NamedRerouteDeclaration(None=Power(
    Base=MaterialFunctionCall(Input=@n64, Input Low=@n65, Input High=@n66) {materialFunction=&RemapValue_0-1}.Result
    Exp=Constant() {r=1.0584}
   )) {name=Bottom Linear Gradient, nodeColor=<rgba 0 0 0 1>}
  n67 = MaterialFunctionCall(
   UVs=MaterialFunctionCall(UVs=NamedRerouteUsage(), Texture Scale=Constant() {r=1.5}) {materialFunction=&ScaleUVsByCenter}.UVs
   GlowMin=Constant() {r=-0.5}
   GlowMax=Constant() {r=1}
  ) {materialFunction=&MF_UI_Circle}
  NamedRerouteDeclaration(None=OneMinus(None=MaterialFunctionCall("Value (0-1)"=@n67.Fill) {materialFunction=&MF_UI_EaseCurves}.EaseInOutSine)) {name=BG Radial Gradient, nodeColor=<rgba 0 0 0 1>}
  NamedRerouteDeclaration(None=@n51.Transition Value) {name=Transition, nodeColor=<rgba 0 0 0 1>}
  NamedRerouteDeclaration(None=Constant3Vector() {constant=<rgba 0.3613 0.6308 0.0382 1>}) {name=ColourCloud_Night_03, nodeColor=<rgba 0 0 0 1>}
  NamedRerouteDeclaration(None=@n50.A) {name=Cloud 01 Alpha Mask, nodeColor=<rgba 0 0 0 1>}
  NamedRerouteDeclaration(None=@n56.Scaled to Fit) {name=ScaledToFit, nodeColor=<rgba 0 0 0 1>}
  NamedRerouteDeclaration(None=Constant3Vector() {constant=<rgba 0.006 0.2747 0.7305 1>}) {name=ColourBG_Day_04, nodeColor=<rgba 0 0 0 1>}
  n68 = Constant() {r=1}
  NamedRerouteDeclaration(None=Multiply(A=Time(), B=@n68)) {name=Global Speed, nodeColor=<rgba 0 0 0 1>}
  NamedRerouteDeclaration(None=@n68) {name=Global Speed Input, nodeColor=<rgba 0 0 0 1>}
  NamedRerouteDeclaration(None=@n56.Scaled to Fill) {name=ScaledToFill, nodeColor=<rgba 0 0 0 1>}
  n69 = MaterialFunctionCall(UVs=NamedRerouteUsage(), Texture Scale=AppendVector(A=Constant() {r=2.1}, B=Constant() {r=1.8})) {materialFunction=&ScaleUVsByCenter}
  n70 = Time()
  n71 = FunctionInput(Preview=Constant() {r=0.005}) {bUsePreviewValueAsDefault=true, inputName=Cloud01Speed, sortPriority=2}
  n72 = AppendVector(A=Frac(None=Multiply(A=@n70, B=@n71)), B=Constant())
  n73 = Multiply(A=Frac(None=Add(A=@n69.UVs, B=@n72)), B=AppendVector(A=Constant() {r=2}, B=Constant() {r=3.3238}))
  NamedRerouteDeclaration(None=MaterialFunctionCall!(UV=Floor(None=@n73), Seed=Constant() {r=4.4898}) {materialFunction=&MF_UI_Random}.Result) {name=Random Foreground Cells New, nodeColor=<rgba 0 0 0 1>}
  NamedRerouteDeclaration(None=Constant3Vector() {constant=<rgba 0.4564 0.7084 0.9387 1>}) {name=ColourBG_Day_01, nodeColor=<rgba 0 0 0 1>}
  NamedRerouteDeclaration(None=Constant3Vector() {constant=<rgba 0.0612 0.4179 0.7991 1>}) {name=ColourBG_Day_02, nodeColor=<rgba 0 0 0 1>}
  n74 = Constant2Vector() {g=0.5, r=0.5}
  n75 = Subtract(A=Frac(None=@n73), B=@n74)
  n76 = AppendVector(A=LinearInterpolate(A=Constant() {r=1.2}, B=Constant() {r=1}, Alpha=NamedRerouteUsage()), B=Constant() {r=1})
  NamedRerouteDeclaration(None=Add(A=Multiply(A=@n75, B=@n76), B=@n74)) {name=Foreground Cloud UV New, nodeColor=<rgba 0 0 0 1>}
  n77 = TextureSample(UVs=@n49) {texture=&T_UI_Quests_CH6MS2_Cloud01_MSDF}
  NamedRerouteDeclaration(None=MaterialFunctionCall(DistanceField=@n77.R, Threshold=Constant() {r=0.28}, Smoothing=Constant()) {materialFunction=&MF_UI_AntiAliasing}.Outer) {name=Cloud 01 R Mask New, nodeColor=<rgba 0 0 0 1>}
  NamedRerouteDeclaration(None=MaterialFunctionCall(DistanceField=@n77.G, Threshold=Constant() {r=0.28}, Smoothing=Constant()) {materialFunction=&MF_UI_AntiAliasing}.Outer) {name=Cloud 01 G Mask New, nodeColor=<rgba 0 0 0 1>}
  NamedRerouteDeclaration(None=MaterialFunctionCall(DistanceField=@n77.B, Threshold=Constant() {r=0.28}, Smoothing=Constant()) {materialFunction=&MF_UI_AntiAliasing}.Outer) {name=Cloud 01 B Mask New, nodeColor=<rgba 0 0 0 1>}
  n78 = MaterialFunctionCall(UVs=NamedRerouteUsage(), Texture Scale=AppendVector(A=Constant() {r=1.2002}, B=Constant() {r=0.9674})) {materialFunction=&ScaleUVsByCenter}
  n79 = Time()
  n80 = FunctionInput(Preview=Constant() {r=0.0171}) {bUsePreviewValueAsDefault=true, inputName=Cloud00Speed, sortPriority=1}
  n81 = AppendVector(A=Frac(None=Multiply(A=@n79, B=@n80)), B=Constant())
  n82 = Multiply(A=Frac(None=Add(A=@n78.UVs, B=@n81)), B=AppendVector(A=Constant() {r=2}, B=Constant() {r=4.5785}))
  NamedRerouteDeclaration(None=MaterialFunctionCall(UV=Floor(None=@n82), Seed=Constant() {r=1.5328}) {materialFunction=&MF_UI_Random}.Result) {name=Random Fore Cells New, nodeColor=<rgba 0 0 0 1>}
  n83 = Constant2Vector() {g=0.5, r=0.5}
  n84 = Subtract(A=Frac(None=@n82), B=@n83)
  n85 = AppendVector(A=LinearInterpolate(A=Constant() {r=-1.1806}, B=Constant() {r=-1.5984}, Alpha=NamedRerouteUsage()), B=Constant() {r=1})
  NamedRerouteDeclaration(None=Add(A=Multiply(A=@n84, B=@n85), B=@n83)) {name=Fore Cloud UV New, nodeColor=<rgba 0 0 0 1>}
  n86 = MaterialFunctionCall(UVs=NamedRerouteUsage(), Texture Scale=AppendVector(A=Constant() {r=2.2641}, B=Constant() {r=2.1273})) {materialFunction=&ScaleUVsByCenter}
  n87 = Time()
  n88 = FunctionInput(Preview=Constant() {r=0.0065}) {bUsePreviewValueAsDefault=true, inputName=Cloud02Speed, sortPriority=3}
  n89 = AppendVector(A=Frac(None=Multiply(A=@n87, B=@n88)), B=Constant())
  n90 = Multiply(A=Frac(None=Add(A=@n86.UVs, B=@n89)), B=AppendVector(A=Constant() {r=2}, B=Constant() {r=4.9973}))
  NamedRerouteDeclaration(None=MaterialFunctionCall(UV=Floor(None=@n90), Seed=Constant() {r=3.5059}) {materialFunction=&MF_UI_Random}.Result) {name=Random Midground Cells New, nodeColor=<rgba 0 0 0 1>}
  n91 = Constant2Vector() {g=0.5, r=0.5}
  n92 = Subtract(A=Frac(None=@n90), B=@n91)
  n93 = AppendVector(A=LinearInterpolate(A=Constant() {r=2.2}, B=Constant() {r=1.5}, Alpha=NamedRerouteUsage()), B=Constant() {r=1})
  NamedRerouteDeclaration(None=Add(A=Multiply(A=@n92, B=@n93), B=@n91)) {name=Midground Cloud UV New, nodeColor=<rgba 0 0 0 1>}
  NamedRerouteDeclaration(None=@n55) {name=ColourBG_Day_03, nodeColor=<rgba 0 0 0 1>}
  n94 = MaterialFunctionCall(UVs=NamedRerouteUsage(), Texture Scale=AppendVector(A=Constant() {r=1.7319}, B=Constant() {r=1.1211})) {materialFunction=&ScaleUVsByCenter}
  n95 = Time()
  n96 = FunctionInput(Preview=Constant() {r=0.015}) {bUsePreviewValueAsDefault=true, inputName=Cloud03Speed, sortPriority=4}
  n97 = AppendVector(A=Frac(None=Multiply(A=@n95, B=@n96)), B=Constant())
  n98 = Multiply(A=Frac(None=Add(A=@n94.UVs, B=@n97)), B=AppendVector(A=Constant() {r=3}, B=Constant() {r=5.851}))
  NamedRerouteDeclaration(None=MaterialFunctionCall(UV=Floor(None=@n98), Seed=Constant() {r=2.7177}) {materialFunction=&MF_UI_Random}.Result) {name=Random Background Cells New, nodeColor=<rgba 0 0 0 1>}
  n99 = Constant2Vector() {g=0.5, r=0.5}
  n100 = Subtract(A=Frac(None=@n98), B=@n99)
  n101 = AppendVector(A=LinearInterpolate(A=Constant() {r=2.4126}, B=Constant() {r=1.7}, Alpha=NamedRerouteUsage()), B=Constant() {r=1})
  NamedRerouteDeclaration(None=Add(A=Multiply(A=@n100, B=@n101), B=@n99)) {name=Background Cloud UV New, nodeColor=<rgba 0 0 0 1>}
  n102 = MaterialFunctionCall(UVs=NamedRerouteUsage(), Texture Scale=AppendVector(A=Constant() {r=2.2123}, B=Constant() {r=0.9476})) {materialFunction=&ScaleUVsByCenter}
  n103 = Time()
  n104 = FunctionInput(Preview=Constant() {r=0.015}) {bUsePreviewValueAsDefault=true, inputName=Cloud04Speed, sortPriority=4}
  n105 = AppendVector(A=Frac(None=Multiply(A=@n103, B=@n104)), B=Constant())
  n106 = Multiply(A=Frac(None=Add(A=@n102.UVs, B=@n105)), B=AppendVector(A=Constant() {r=6}, B=Constant() {r=10}))
  NamedRerouteDeclaration(None=MaterialFunctionCall(UV=Floor(None=@n106), Seed=Constant() {r=2.2979}) {materialFunction=&MF_UI_Random}.Result) {name=Random Far Background Cells New, nodeColor=<rgba 0 0 0 1>}
  n107 = Constant2Vector() {g=0.5, r=0.5}
  n108 = Subtract(A=Frac(None=@n106), B=@n107)
  n109 = AppendVector(A=LinearInterpolate(A=Constant() {r=1.5}, B=Constant() {r=2.4}, Alpha=NamedRerouteUsage()), B=Constant() {r=1})
  NamedRerouteDeclaration(None=Add(A=Multiply(A=@n108, B=@n109), B=@n107)) {name=Far Background Cloud UV New, nodeColor=<rgba 0 0 0 1>}
  MaterialExpressionAdd_31 = Add(A=NamedRerouteUsage(), B=AppendVector(A=Constant(), B=Constant() {r=0.21}))
  n110 = Convert(None=@MaterialExpressionAdd_31) {
   nodeName=Break Float2
   convertInputs [{
     type=Vector2
     expressionInput {
      expression=&self:MaterialExpressionAdd_31
      inputName=None
      mask=0
      maskA=0
      maskB=0
      maskG=0
      maskR=0
      outputIndex=0
     }
    }]
   convertOutputs [{type=Scalar}, {type=Scalar}]
  }
  n111 = TextureSample(UVs=AppendVector(A=DDX(Value=@n110), B=DDY(Value=@n110))) {texture=&T_UI_Quests_CH6MS2_Cloud01_MSDF}
  NamedRerouteDeclaration(None=MaterialFunctionCall(DistanceField=@n111.R, Threshold=Constant() {r=0.28}, Smoothing=Constant()) {materialFunction=&MF_UI_AntiAliasing}.Outer) {name=Cloud 00 R Mask New, nodeColor=<rgba 0 0 0 1>}
  NamedRerouteDeclaration(None=MaterialFunctionCall(DistanceField=@n111.G, Threshold=Constant() {r=0.28}, Smoothing=Constant()) {materialFunction=&MF_UI_AntiAliasing}.Outer) {name=Cloud 00 G Mask New, nodeColor=<rgba 0 0 0 1>}
  NamedRerouteDeclaration(None=MaterialFunctionCall(DistanceField=@n111.B, Threshold=Constant() {r=0.28}, Smoothing=Constant()) {materialFunction=&MF_UI_AntiAliasing}.Outer) {name=Cloud 00 B Mask New, nodeColor=<rgba 0 0 0 1>}
  n112 = TextureSample(UVs=@MaterialExpressionAdd_32) {texture=/FortniteCoreUI/UI/Quests/CH6MS2/Textures/T_UI_Quests_CH6MS2_Cloud02_MSDF1}
  NamedRerouteDeclaration(None=MaterialFunctionCall(DistanceField=@n112.R, Threshold=Constant() {r=0.15}, Smoothing=Constant()) {materialFunction=&MF_UI_AntiAliasing}.Outer) {name=Cloud 02 R Mask New, nodeColor=<rgba 0 0 0 1>}
  NamedRerouteDeclaration(None=MaterialFunctionCall(DistanceField=@n112.G, Threshold=Constant() {r=0.15}, Smoothing=Constant()) {materialFunction=&MF_UI_AntiAliasing}.Outer) {name=Cloud 02 G Mask New, nodeColor=<rgba 0 0 0 1>}
  NamedRerouteDeclaration(None=MaterialFunctionCall(DistanceField=@n112.B, Threshold=Constant() {r=0.334}, Smoothing=Constant()) {materialFunction=&MF_UI_AntiAliasing}.Outer) {name=Cloud 02 B Mask New, nodeColor=<rgba 0 0 0 1>}
  NamedRerouteDeclaration(None=Constant3Vector() {constant=<rgba 1 1 1 0>}) {name=ColourCloud_Day_01, nodeColor=<rgba 0 0 0 1>}
  n113 = TextureSample(UVs=@MaterialExpressionAdd_33) {texture=/FortniteCoreUI/UI/Quests/CH6MS2/Textures/T_UI_Quests_CH6MS2_Cloud03_MSDF1}
  NamedRerouteDeclaration(None=MaterialFunctionCall(DistanceField=@n113.R, Threshold=Constant() {r=0.25}, Smoothing=Constant()) {materialFunction=&MF_UI_AntiAliasing}.Outer) {name=Cloud 03 R Mask New, nodeColor=<rgba 0 0 0 1>}
  NamedRerouteDeclaration(None=MaterialFunctionCall(DistanceField=@n113.G, Threshold=Constant() {r=0.25}, Smoothing=Constant()) {materialFunction=&MF_UI_AntiAliasing}.Outer) {name=Cloud 03 G Mask New, nodeColor=<rgba 0 0 0 1>}
  NamedRerouteDeclaration(None=MaterialFunctionCall(DistanceField=@n113.B, Threshold=Constant() {r=0.25}, Smoothing=Constant()) {materialFunction=&MF_UI_AntiAliasing}.Outer) {name=Cloud 03 B Mask New, nodeColor=<rgba 0 0 0 1>}
  n114 = TextureSample(UVs=@MaterialExpressionAdd_34) {texture=/FortniteCoreUI/UI/Quests/CH6MS2/Textures/T_UI_Quests_CH6MS2_Cloud04_MSDF1}
  NamedRerouteDeclaration(None=MaterialFunctionCall(DistanceField=@n114.R, Threshold=Constant() {r=0.25}, Smoothing=Constant()) {materialFunction=&MF_UI_AntiAliasing}.Outer) {name=Cloud 04 R Mask New, nodeColor=<rgba 0 0 0 1>}
  NamedRerouteDeclaration(None=MaterialFunctionCall(DistanceField=@n114.G, Threshold=Constant() {r=0.25}, Smoothing=Constant()) {materialFunction=&MF_UI_AntiAliasing}.Outer) {name=Cloud 04 G Mask New, nodeColor=<rgba 0 0 0 1>}
  NamedRerouteDeclaration(None=MaterialFunctionCall(DistanceField=@n114.B, Threshold=Constant() {r=0.25}, Smoothing=Constant()) {materialFunction=&MF_UI_AntiAliasing}.Outer) {name=Cloud 04 B Mask New, nodeColor=<rgba 0 0 0 1>}
  NamedRerouteDeclaration(None=Constant3Vector() {constant=<rgba 0.6724 0.855 0.9911 1>}) {name=ColourCloud_Day_02, nodeColor=<rgba 0 0 0 1>}
  NamedRerouteUsage()
  NamedRerouteUsage()
  NamedRerouteUsage()
  NamedRerouteUsage()
  NamedRerouteUsage()
  NamedRerouteUsage()
  NamedRerouteUsage()
  NamedRerouteUsage()
  ScalarParameter() {defaultValue=0.4, parameterName=BaseBGNight_LinearGradientLow}
  ScalarParameter() {defaultValue=1, parameterName=BaseBGNight_LinearGradientHigh}
 ]
}
