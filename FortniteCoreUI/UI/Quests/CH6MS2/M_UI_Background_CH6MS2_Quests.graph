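# UI material that renders the CH6MS2 Quests background, with translucent blending.
# It is a thin wrapper that delegates the visuals to the MF_UI_Background_CH6MS2_Quests material function.
# The function's BGLocalUV input is fed by a constant, and its five cloud speed inputs are left unconnected.
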Material /FortniteCoreUI/UI/Quests/CH6MS2/M_UI_Background_CH6MS2_Quests
graph {
 properties {
  bAutomaticallySetUsageInEditor=false
  bUsedWithStaticMesh=false
  blendMode=BLEND_Translucent
  materialDomain=MD_UI
  usedShadingModels=Default Lit
 }
 nodes [MaterialFunctionCall!(BGLocalUV=Constant() {r=0.5}, Cloud00Speed, Cloud01Speed, Cloud02Speed, Cloud04Speed, Cloud03Speed) {materialFunction=&/MF_UI_Background_CH6MS2_Quests}]
}
